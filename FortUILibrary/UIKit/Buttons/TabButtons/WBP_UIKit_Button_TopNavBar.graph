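# Top navigation bar tab button, built on the UIKit generic modular button base.
# PreConstruct switches text layout, font materials and background corner radius between mobile and other platforms, and sets selection timings.
# SetData caches the FortTabViewModel, whose linked UI feature and force-image flag drive text and icon, with V-Buck special handling for the store tab.
# Hover passes through only for mouse and keyboard, and gamepad selection plays a sound.
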WidgetBlueprint /FortUILibrary/UIKit/Buttons/TabButtons/WBP_UIKit_Button_TopNavBar
graphs [
 EventGraph {nodes [
   n1 = Event PreConstruct()
   > SetTextLayout()
   > SetFontStyle()
   > Parent: PreConstruct(IsDesignTime=@n1.IsDesignTime)
   > SetBlocksTiming!(
    execute
    self:UIKit Modular Button Object Reference
    Timing:"UIKit Block Timing Structure (by ref)"=Make UIKit Block Timing!(
     Hovering:float=0
     Unhovering:float=0
     Focusing:float=0
     Unfocusing:float=0
     Pressing:float=0
     Releasing:float=0
     Disabling:float=0
     Enabling:float=0
     Selecting:float=0.15
     DeselectingIdle:float=0.1
     DeselectingFocused:float=0.1
    ) -> (UIKitBlockTiming:UIKit Block Timing Structure) {is=title K2Node_MakeStruct}
   ) -> (then)
   > n2 = Branch(Condition=IsMobileGame())
   > n3 = Cast To WBP_UIKit_Block_BackgroundTopNavBar!(execute, Object=Get Block_Background() -> (Block_Background:WBP UIKit Block Background Base Object Reference)) -> (then, CastFailed, AsWBP UIKit Block Background Top Nav Bar:WBP UIKit Block Background Top Nav Bar Object Reference) {is=title K2Node_DynamicCast}
   > SetCornerRadiuses!(execute, self:WBP UIKit Block Background Top Nav Bar Object Reference=@n3, CornerRadius:float=24, RadiusInPixels:bool=true) -> (then)
   n4 = Cast To WBP_UIKit_Block_BackgroundTopNavBar(execute=@n2.else, Object=Get Block_Background())
   > SetCornerRadiuses(self=@n4, CornerRadius=12, RadiusInPixels=true)
   Event Construct()
   > Parent: Construct()
   Event On Widget Styled() {is=title K2Node_Event}
   > Parent: On Widget Styled() {is=title K2Node_CallParentFunction}
   Event On Hovered()
   n5 = "Equal (Enum)"(A=GetCurrentInputType(self=Get CommonInputSubsystem()), B=MouseAndKeyboard)
   > Branch(Condition=@n5)
   > Parent: On Hovered() {is=title K2Node_CallParentFunction}
   Event On Unhovered()
   > Branch(Condition=@n5)
   > Parent: On Unhovered() {is=title K2Node_CallParentFunction}
   Event On Selected()
   > Parent: On Selected() {is=title K2Node_CallParentFunction}
   n6 = Get CommonInputSubsystem()
   > Branch(Condition="Equal (Enum)"(A=GetCurrentInputType(self=@n6), B=Gamepad))
   > PlaySound2D(Sound=/Game/Sounds/Fort_UI_Sounds/FrontEnd/Tabs/MSS_UI_Button_Tab_Pressed_Generic)
  ]}
 SetBangType {nodes [
   n1 = SetBangType!() -> (then, EBangType:EFortBangType Enum=Invalid) {is=title K2Node_FunctionEntry}
   > n2 = Get!(execute:Exec) -> (then:Exec, else:Exec, StatusIndicator:WBP UIKit Status Indicator Object Reference)
   > PassBangTypeFromBP!(execute, self:WBP UIKit Status Indicator Base Object Reference=@n2, EBangType:EFortBangType Enum=@n1) -> (then)
  ]}
 SetData {nodes [
   n1 = SetData!() -> (then, VM Button:UI Kit Selectable Button View Model Object Reference) {is=title K2Node_FunctionEntry}
   > Parent: SetData!(execute, VM Button:UI Kit Selectable Button View Model Object Reference=@n1) -> (then) {is=title K2Node_CallParentFunction}
   > n2 = Cast To FortTabViewModel!(execute, Object=@n1) -> (then, CastFailed, AsFort Tab Viewmodel:Fort Tab Viewmodel Object Reference) {is=title K2Node_DynamicCast}
   > Set FortTabViewModel(FortTabViewModel:Fort Tab Viewmodel Object Reference=@n2) -> (Output_Get:Fort Tab Viewmodel Object Reference)
  ]}
 SetFontStyle {nodes [
   SetFontStyle() {is=title K2Node_FunctionEntry}
   n1 = IsMobileGame()
   n2 = Get!(None:%=Self-Reference()) -> (
    Font_FontObject:Object Reference
    Font_FontMaterial:Object Reference
    Font_OutlineSettings_OutlineSize:int
    Font_OutlineSettings_bMiteredCorners:bool=~
    Font_OutlineSettings_bSeparateFillAlpha:bool=~
    Font_OutlineSettings_bApplyOutlineToDropShadows:bool=~
    Font_OutlineSettings_OutlineMaterial:Object Reference
    Font_OutlineSettings_OutlineColor:Linear Color Structure
    Font_TypefaceFontName:Name
    Font_Size:float=~
    Font_LetterSpacing:int
    Font_SkewAmount:float=~
    Font_bForceMonospaced:bool=~
    Font_bMaterialIsStencil:bool=~
    Font_MonospacedWidth:float=~
   )
   > SetFont!(
    execute
    self:WBP UIKit Button Generic Base Object Reference=Self-Reference()
    In Font_FontObject:Object Reference=@n2
    In Font_FontMaterial:Object Reference=Select(
     Option 0:Object Reference=Get TextMaterial() -> (TextMaterial:Material Instance Object Reference)
     Option 1:Object Reference=Get MobileTextMaterial() -> (MobileTextMaterial:Material Instance Object Reference)
     Index=@n1
    ) -> (ReturnValue:Object Reference)
    In Font_OutlineSettings_OutlineSize:int=@n2.Font_OutlineSettings_OutlineSize
    In Font_OutlineSettings_bMiteredCorners:bool=@n2.Font_OutlineSettings_bMiteredCorners
    In Font_OutlineSettings_bSeparateFillAlpha:bool=@n2.Font_OutlineSettings_bSeparateFillAlpha
    In Font_OutlineSettings_bApplyOutlineToDropShadows:bool=@n2.Font_OutlineSettings_bApplyOutlineToDropShadows
    In Font_OutlineSettings_OutlineMaterial:Object Reference=Select(
     Option 0=Get TextOutlineMaterial() -> (TextOutlineMaterial:Material Instance Object Reference)
     Option 1=Get MobileTextOutlineMaterial() -> (MobileTextOutlineMaterial:Material Instance Object Reference)
     Index=@n1
    )
    In Font_OutlineSettings_OutlineColor:Linear Color Structure=@n2.Font_OutlineSettings_OutlineColor
    In Font_TypefaceFontName:Name=@n2.Font_TypefaceFontName
    In Font_Size:float=@n2.Font_Size
    In Font_LetterSpacing:int=@n2.Font_LetterSpacing
    In Font_SkewAmount:float=@n2.Font_SkewAmount
    In Font_bForceMonospaced:bool=@n2.Font_bForceMonospaced
    In Font_bMaterialIsStencil:bool=false
    In Font_MonospacedWidth:float=@n2.Font_MonospacedWidth
   ) -> (then)
  ]}
 SetIcon {nodes [
   n1 = SetIcon!() -> (then, Icon:Object Reference) {is=title K2Node_FunctionEntry}
   > Get!(execute:Exec) -> (then:Exec, else:Exec, FortTabViewModel:Fort Tab Viewmodel Object Reference)
   > Set UIFeature(UIFeature:EFortUIFeature Enum=GetLinkedUIFeature!(self:Fort Tab Viewmodel Object Reference=Get FortTabViewModel() -> (FortTabViewModel:Fort Tab Viewmodel Object Reference)) -> (ReturnValue:EFortUIFeature Enum=ShowHome)) -> (Output_Get:EFortUIFeature Enum=ShowHome)
   n2 = Get FortTabViewModel()
   > n3 = Branch(Condition=OR Boolean(A=IsMobileGame(), B=GetForceImage!(self:Fort Tab Viewmodel Object Reference=@n2) -> (ReturnValue:bool)))
   > Parent: SetIcon!(execute, Icon:Object Reference=@n1) -> (then) {is=title K2Node_CallParentFunction}
   n4 = Branch(
    execute=@n3.else
    Condition="Equal (Enum)"(A:EFortUIFeature Enum=Get UIFeature() -> (UIFeature:EFortUIFeature Enum=ShowHome), B:EFortUIFeature Enum=ShowMainStore)
   )
   > SetIconPadding!(execute, self:WBP UIKit Button Generic Base Object Reference, Padding:Margin Structure=Make Margin(Left=0, Top=0, Right=4, Bottom=0)) -> (then)
   > Parent: SetIcon(Icon=Get IconVBuck() -> (IconVBuck:Texture 2D Object Reference))
   Parent: SetIcon(execute=@n4.else)
  ]}
 SetText {nodes [
   n1 = SetText!() -> (then, Text:Text) {is=title K2Node_FunctionEntry}
   > n2 = Get()
   > Set UIFeature(UIFeature=GetLinkedUIFeature(self=@n2) -> (ReturnValue=ShowHome)) -> (Output_Get=ShowHome)
   > n3 = Branch(Condition=IsMobileGame())
   n4 = GetTabDisplayName!(self:Fort Tab Viewmodel Object Reference=Get FortTabViewModel()) -> (ReturnValue:Text)
   > Parent: SetText!(execute, Text:Text=SelectText(A=Get Text!() -> (Text), B=@n4, bPickA=TextIsEmpty(InText=@n4))) -> (then) {is=title K2Node_CallParentFunction}
   n5 = Get UIFeature() -> (UIFeature=ShowHome)
   > Set SecondText(SecondText:Text=SelectText(A=@n1, bPickA="Equal (Enum)"(A=@n5, B=ShowMainStore))) -> (Output_Get:Text)
   > UpdateSecondText(self:WBP UIKit Button Generic Base Object Reference)
   n6 = Branch(execute=@n3.else, Condition=GetForceImage(self=@n2))
   > Parent: SetText()
   Parent: SetText(execute=@n6.else, Text=@n1)
  ]}
 SetTextLayout {nodes [
   SetTextLayout() {is=title K2Node_FunctionEntry}
   > n1 = Set IsMobile(IsMobile:bool=IsMobileGame()) -> (Output_Get:bool)
   > n2 = Branch(Condition=@n1)
   > Set TextBlockAlignment(TextBlockAlignment:E_UI_Alignment Enum=NewEnumerator1) -> (Output_Get:E_UI_Alignment Enum=NewEnumerator0)
   > Set IsTextOutsideBackground(IsTextOutsideBackground:bool=true) -> (Output_Get:bool)
   > Set ShouldTextImpactLayout(ShouldTextImpactLayout:bool=false) -> (Output_Get:bool)
   > Set!(execute, TextPadding_Left:float=-200, TextPadding_Top:float=0, TextPadding_Right:float=-200, TextPadding_Bottom:float=0) -> (then, Output_Get:Margin Structure)
   > Set!(execute, TextNudge_X:double=0, TextNudge_Y:double=26) -> (then, Output_Get:Vector 2D Structure)
   > Set UseSecondTextBlock(UseSecondTextBlock:bool=true) -> (Output_Get:bool)
   > Set SecondFont(SecondFont:Slate Font Info Structure=Get Font() -> (Font:Slate Font Info Structure)) -> (Output_Get:Slate Font Info Structure)
   > Set!(execute, SecondTextPadding_Left:float=0, SecondTextPadding_Top:float=0, SecondTextPadding_Right:float=0, SecondTextPadding_Bottom:float=0) -> (then, Output_Get:Margin Structure)
   > Set SecondText()
   Set TextBlockAlignment(execute=@n2.else, TextBlockAlignment=NewEnumerator3) -> (Output_Get=NewEnumerator0)
   > Set IsTextOutsideBackground(IsTextOutsideBackground=false)
   > Set ShouldTextImpactLayout(ShouldTextImpactLayout=true)
   > Set!(execute, TextPadding_Left:float=0, TextPadding_Top:float=0, TextPadding_Right:float=0, TextPadding_Bottom:float=0) -> (then, Output_Get)
   > Set!(execute, TextNudge_X:double=0, TextNudge_Y:double=0) -> (then, Output_Get:Vector 2D Structure)
  ]}
]
variables [
 Block_Text_Secondary {type=WBP_UIKit_Block_Text_C, default=null}
 IconVBuck {type=Texture2D, default=/FortUILibrary/Textures/Icons/T_UI_Icon_UIKit_VBuck_SDF}
 IsMobile {type=boolean}
 MobileTextMaterial {type=MaterialInstance, default=/FortUILibrary/UIKit/Buttons/TabButtons/_MaterialInstances/MI_UI_TopNav_TextMobile}
 MobileTextOutlineMaterial {type=MaterialInstance, default=/FortUILibrary/UIKit/Buttons/TabButtons/_MaterialInstances/MI_UI_TopNav_TextOutlineMobile}
 TextMaterial {type=MaterialInstance, default=/FortUILibrary/UIKit/Buttons/TabButtons/_MaterialInstances/MI_UI_TopNav_Text}
 TextOutlineMaterial {type=MaterialInstance, default=/FortUILibrary/UIKit/Buttons/TabButtons/_MaterialInstances/MI_UI_TopNav_TextOutline}
]
widgets {
 info {
  inheritedWidgetCount=0
  parentClass=/FortUILibrary/Systems/ModularButton/WBP_UIKit_ButtonGeneric_Base.WBP_UIKit_ButtonGeneric_Base_C
  rootWidgetClass=null
  widgetCount=0
 }
 widgets []
}
