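# Tab button widget derived from WBP_UIKit_Button_Selectable, adding no widgets of its own, for side-nav and top-nav tabs.
# PreConstruct sets selection-only block timings, then uses IsMobile to pick text and outline materials and corner radii.
# Casts route the radius to either the TabSideNav or TopNavBar background block.
# Overrides skip parent press handling when already selected on gamepad, and parent hover/unhover handling on touch.
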WidgetBlueprint /FortUILibrary/UIKit/Buttons/TabButtons/WBP_UIKit_Button_Tab
graphs [EventGraph {nodes [
   n1 = Event PreConstruct()
   > Parent: PreConstruct(IsDesignTime=@n1.IsDesignTime)
   > SetBlocksTiming!(
    execute
    self:UIKit Modular Button Object Reference
    Timing:"UIKit Block Timing Structure (by ref)"=Make UIKit Block Timing!(
     Hovering:float=0
     Unhovering:float=0
     Focusing:float=0
     Unfocusing:float=0
     Pressing:float=0
     Releasing:float=0
     Disabling:float=0
     Enabling:float=0
     Selecting:float=0.15
     DeselectingIdle:float=0.1
     DeselectingFocused:float=0.1
    ) -> (UIKitBlockTiming:UIKit Block Timing Structure) {is=title K2Node_MakeStruct}
   ) -> (then)
   > n2 = Branch(Condition=Get IsMobile() -> (IsMobile:bool))
   n3 = Get Block_Text() -> (Block_Text:WBP UIKit Block Text Object Reference)
   > SetFontMaterial!(
    execute
    self:WBP UIKit Block Text Object Reference=@n3
    In Material:Material Instance Object Reference=Get MobileTextMaterial() -> (MobileTextMaterial:Material Instance Object Reference)
   ) -> (then)
   > SetOutlineMaterial!(
    execute
    self:WBP UIKit Block Text Object Reference=@n3
    In Material:Material Instance Object Reference=Get MobileTextOutlineMaterial() -> (MobileTextOutlineMaterial:Material Instance Object Reference)
   ) -> (then)
   n4 = Get Block_Background() -> (Block_Background:WBP UIKit Block Background Base Object Reference)
   > n5 = Cast To WBP_UIKit_Block_BackgroundTabSideNav!(execute, Object=@n4) -> (then, CastFailed, AsWBP UIKit Block Background Tab Side Nav:WBP UIKit Block Background Tab Side Nav Object Reference) {is=title K2Node_DynamicCast}
   > SetCornerRadiuses!(execute, self:WBP UIKit Block Background Tab Side Nav Object Reference=@n5, CornerRadius:float=20, RadiusInPixels:bool=true) -> (then)
   n6 = Cast To WBP_UIKit_Block_BackgroundTopNavBar!(execute=@n5.CastFailed, Object=@n4) -> (then, CastFailed, AsWBP UIKit Block Background Top Nav Bar:WBP UIKit Block Background Top Nav Bar Object Reference) {is=title K2Node_DynamicCast}
   > SetCornerRadiuses(self:WBP UIKit Block Background Top Nav Bar Object Reference=@n6, CornerRadius=20, RadiusInPixels=true)
   n7 = Get Block_Text()
   SetFontMaterial(execute=@n2.else, self=@n7, In Material=Get TextMaterial() -> (TextMaterial:Material Instance Object Reference))
   > SetOutlineMaterial(self=@n7, In Material=Get TextOutlineMaterial() -> (TextOutlineMaterial:Material Instance Object Reference))
   > n8 = Cast To WBP_UIKit_Block_BackgroundTabSideNav(Object=@n4)
   > SetCornerRadiuses(self:WBP UIKit Block Background Tab Side Nav Object Reference=@n8, CornerRadius=12, RadiusInPixels=true)
   n9 = Cast To WBP_UIKit_Block_BackgroundTopNavBar(execute=@n8.CastFailed, Object=@n4)
   > SetCornerRadiuses(self:WBP UIKit Block Background Top Nav Bar Object Reference=@n9, CornerRadius=12, RadiusInPixels=true)
   Event On Pressed() {is=title K2Node_Event}
   n10 = GetSelected!(self:Common Button Base Object Reference) -> (ReturnValue:bool)
   n11 = "Equal (Enum)"(A=Get CurrentInputType() -> (CurrentInputType:ECommonInputType Enum), B=Gamepad)
   > Branch(Condition=NOT Boolean(A=AND Boolean(A=@n10, B=@n11)))
   > Parent: On Pressed() {is=title K2Node_CallParentFunction}
   Event On Hovered()
   n12 = IsUsingTouch(PlayerController=GetOwningPlayer())
   > n13 = Branch(Condition=@n12)
   Parent: On Hovered(execute=@n13.else) {is=title K2Node_CallParentFunction}
   Event On Unhovered()
   > n14 = Branch(Condition=@n12)
   Parent: On Unhovered(execute=@n14.else) {is=title K2Node_CallParentFunction}
  ]}]
variables [
 IsMobile {type=boolean}
 MobileTextMaterial {type=MaterialInstance, default=/FortUILibrary/UIKit/Buttons/TabButtons/_MaterialInstances/MI_UI_TopNav_TextMobile}
 MobileTextOutlineMaterial {type=MaterialInstance, default=/FortUILibrary/UIKit/Buttons/TabButtons/_MaterialInstances/MI_UI_TopNav_TextOutlineMobile}
 TextMaterial {type=MaterialInstance, default=/FortUILibrary/UIKit/Buttons/TabButtons/_MaterialInstances/MI_UI_TopNav_Text}
 TextOutlineMaterial {type=MaterialInstance, default=/FortUILibrary/UIKit/Buttons/TabButtons/_MaterialInstances/MI_UI_TopNav_TextOutline}
]
widgets {
 info {
  inheritedWidgetCount=0
  parentClass=/FortUILibrary/UIKit/Buttons/WBP_UIKit_Button_Selectable.WBP_UIKit_Button_Selectable_C
  rootWidgetClass=null
  widgetCount=0
 }
 widgets []
}
